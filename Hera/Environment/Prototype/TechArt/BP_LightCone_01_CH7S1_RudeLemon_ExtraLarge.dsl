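; Light cone Blueprint that changes its cone length according to the player's distance from it.
; On BeginPlay, where cosmetic events can run, it caches player pawn 0 and the default length, then starts a looping half-second timer instead of updating on Tick.
; Each timer call blends from NewConeLength to the default across FadeMin to FadeMax.
; SetConeScale then derives end radius from the cone angle and rescales the cone to match.
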(event EventTick
  (params (DeltaSeconds :type Float%20%28single-precision%29))
  (|CallParentFunction DeltaSeconds))

(event Collision|EventActorBeginOverlap
  (params (OtherActor :type Actor%20Object%20Reference))
  (|CallParentFunction OtherActor) ; This node is disabled and will not be called.\nDrag off pins to build functionality.
)

(event EventBeginPlay
  (|CallParentFunction)
  (Networking|CanExecuteCosmeticEvents
    (:True
      (Variables|Default|SetPlayerPawn (Game|GetPlayerPawn 0))
      (Variables|Default|SetDefaultConeLength (Variables|ConeShape|GetCone%20Length))
      (Utilities|Time|SetTimerbyEvent (delegate Custom|CustomEvent) 0.5 true))))

(event Custom|CustomEvent
  (Variables|ConeShape|SetCone%20Length (Math|Float|Lerp (Variables|Default|GetNewConeLength) (Variables|Default|GetDefaultConeLength) (Math|Float|MapRangeClamped (Math|Vector|Distance(Vector) (Transformation|GetActorLocation (Variables|Default|GetPlayerPawn)) (Transformation|GetActorLocation)) (Variables|Default|GetFadeMin) (Variables|Default|GetFadeMax) 0.0 1.0)))
  (CallFunction|SetConeScale))

(fn SetConeScale
  (bind (_x _y) (Math|Vector2D|BreakVector2D (Variables|ConeShape|GetXY%20Scale)))
  (Variables|Default|SetEnd%20Radius (+ (* (Variables|ConeShape|GetCone%20Length) (Math|Trig|Tan(Degrees) (Variables|ConeShape|GetCone%20Angle))) (Variables|ConeShape|GetStart%20Radius)))
  (Transformation|SetWorldScale3D (Variables|Default|GetCone) (Math|Vector|MakeVector (/ (* _x (Variables|Default|GetEnd%20Radius)) 2e+02) (/ (* _y (Variables|Default|GetEnd%20Radius)) 2e+02) (/ (Variables|ConeShape|GetCone%20Length) 1e+03))))

(fn ConstructionScript
  (|CallParentFunction))
