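# Unlit, additive, two-sided material that draws animated electric beams, previewed on a combined beam mesh and enabled for spline and skeletal meshes.
# Two panning noise layers, Amp and Micro, sum into UV distortion for a GeneratedBand arc, while vertex RGB channels form per-beam masks for variation.
# The band drives depth-faded opacity and colored, eye-adaptation-compensated emissive.
# SplineThicken supplies world position offset; some branches, such as HeightLerp, feed nothing.
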Material /Hera/Environment/Prototype/TechArt/M_Electricity
graph {
 properties {
  bAutomaticallySetUsageInEditor=false
  bUsedWithMeshDeformer=true
  bUsedWithSkeletalMesh=true
  bUsedWithSplineMeshes=true
  blendMode=BLEND_Additive
  previewMesh=&/ElectricBeams_Combined
  shadingModel=MSM_Unlit
  twoSided=true
  usedShadingModels=Unlit
 }
 nodes [
  Constant() {r=1.5}
  CurveAtlasRowParameter() {
   atlas=&/CA_TechArtCH7Proto
   curve=&/CC_ParticleAccelerator
   parameterName=ColorCurve
  }
  n1 = LinearInterpolate(
   A=ScalarParameter() {defaultValue=0.06, parameterName=Electricity_Width_Min}
   B=ScalarParameter() {defaultValue=0.15, parameterName=Electricity_Width_Max}
   Alpha=NamedRerouteUsage()
  )
  LinearInterpolate(A=Multiply(A=@n1) {constB=2}, B=@n1, Alpha=NamedRerouteUsage())
  n2 = ScalarParameter() {parameterName=Electricity_Sharpness}
  n3 = Constant() {r=0.5}
  n4 = Add(A=NamedRerouteUsage(), B=NamedRerouteUsage())
  n5 = Saturate(None=Power(
    Base=MaterialFunctionCall!(Width=@n1, Sharpness=@n2, Offset, Direction Switch, Compare=@n3, Input Coordinates=@n4) {materialFunction=/Engine/Functions/Engine_MaterialFunctions03/Procedurals/GeneratedBand}.Result
    Exp=ScalarParameter() {defaultValue=1, parameterName=Amp_Exp}
   ) {constExponent=5})
  n6 = NamedRerouteUsage()
  n7 = ComponentMask(None=VectorParameter() {parameterName=HeightLerpNoiseTiling, defaultValue=<rgba 0.01 0.01 0 1>}.RGB) {g=true, r=true}
  n8 = ScalarParameter() {defaultValue=0.001, parameterName=HeightLerpNoiseDistortionAmount}
  n9 = NamedRerouteUsage()
  n10 = TextureSample(UVs=Panner(Coordinate=Multiply(A=@n9)) {speedY=0.25}) {samplerType=SAMPLERTYPE_Grayscale, texture=/Game/UIKit/Buttons/Textures/LowResBlurredNoise}
  n11 = Panner(
   Coordinate=Add(A=Multiply(A=@n6, B=@n7), B=Multiply(A=@n8, B=@n10.R))
   Speed=ComponentMask(None=VectorParameter() {parameterName=HeightLerpNoiseSpeed, defaultValue=<rgba 0.1 -0.5 0 1>}.RGB) {g=true, r=true}
  ) {speedX=0.1, speedY=-0.5}
  MaterialFunctionCall(
   A=Constant()
   B=Constant() {r=1}
   Transition Phase=@n5
   Height Texture=Multiply(
    A=TextureSample(UVs=@n11) {texture=/Game/Creative/PostProcess/Textures/Cell_Noise_T1}.R
    B=ScalarParameter() {parameterName=HeightLerpNoiseStrength}
   )
   Contrast=Constant() {r=0.5}
  ) {materialFunction=&HeightLerp}
  n12 = TextureCoordinate()
  n13 = Noise!(World Position=AppendVector(A=Panner() {speedY=0.1}, B=Constant()), FilterWidth) {noiseFunction=NOISEFUNCTION_GradientTex, outputMax=0.1, outputMin=-0.1}
  Multiply(A=MaterialFunctionCall(Width=Constant() {r=0.01}, Input Coordinates=Add(A=@n12, B=@n13)) {materialFunction=&GeneratedBand}.Result)
  n14 = LinearInterpolate(
   A=ScalarParameter() {defaultValue=1, parameterName=Electricity_Opacity_Min}
   B=ScalarParameter() {defaultValue=0.3, parameterName=Electricity_Opacity_Max}
   Alpha=NamedRerouteUsage()
  )
  NamedRerouteDeclaration(None=DepthFade(Opacity=Multiply(A=@n14, B=@n5), FadeDistance=ScalarParameter() {defaultValue=200, parameterName=Electricity_DepthFadeDistance})) {name=Electricity_Opacity, nodeColor=<rgba 1 0.0703 0 1>}
  MaterialExpressionTextureCoordinate_5 = TextureCoordinate()
  n15 = Convert(None=@MaterialExpressionTextureCoordinate_5) {
   nodeName=Break Float2
   convertInputs [{
     type=Vector2
     expressionInput {
      expression=&self:MaterialExpressionTextureCoordinate_5
      inputName=None
      mask=0
      maskA=0
      maskB=0
      maskG=0
      maskR=0
      outputIndex=0
     }
    }]
   convertOutputs [{type=Scalar}, {type=Scalar}, {type=Vector2}]
  }
  NamedRerouteDeclaration(None=StaticSwitchParameter(True=@n15, False=@MaterialExpressionTextureCoordinate_5) {defaultValue=true, parameterName=FlipUV}) {name=TexCoord, nodeColor=<rgba 1 0.3516 0 1>}
  n16 = ScalarParameter() {defaultValue=50, parameterName=RibbonThickness}
  NamedRerouteDeclaration(None=Add(
    A=Constant3Vector() {constant=<rgba 0 0 10 0>}
    B=MaterialFunctionCall(WidthBase=@n16, WidthTip=@n16) {materialFunction=&SplineThicken}.WorldPositionOffset
   )) {name=WPO, nodeColor=<rgba 0 1 0.7188 1>}
  n17 = NamedRerouteUsage()
  n18 = ComponentMask(None=VectorParameter() {parameterName=Amp_NoiseTiling, defaultValue=<rgba 0.01 0.005 0 1>}.RGB) {g=true, r=true}
  n19 = LinearInterpolate(A=@n18, B=Multiply(A=@n18) {constB=1.5}, Alpha=NamedRerouteUsage())
  n20 = Constant()
  n21 = ScalarParameter() {defaultValue=0.7, parameterName=Amp_NoisePanningSpeed}
  n22 = TextureSample(UVs=Panner(Coordinate=Multiply(A=@n17, B=@n19), Speed=AppendVector(A=@n20, B=@n21)) {speedY=3}) {samplerType=SAMPLERTYPE_Masks, texture=/Game/Weapons/WeaponSkins/Wraps/Textures/Mash/Mash_Noise}
  n23 = Multiply(
   A=ConstantBiasScale(None=ComponentMask(None=@n22.RGB) {g=true, r=true}) {bias=-0.5, scale=2}
   B=ScalarParameter() {defaultValue=-0.3, parameterName=Amp_NoiseStrength}
  ) {constB=0.1}
  n24 = NamedRerouteUsage()
  n25 = ComponentMask(None=VectorParameter() {parameterName=Micro_NoiseTiling, defaultValue=<rgba 0.1 0.01 0 1>}.RGB) {g=true, r=true}
  n26 = ScalarParameter() {defaultValue=-1, parameterName=Micro_NoisePanningSpeed}
  n27 = Constant()
  n28 = TextureSample(
   UVs=Panner(Coordinate=Multiply(A=@n24, B=@n25), Speed=AppendVector(A=@n26, B=@n27)) {speedY=0.25}
   Tex=TextureObject() {samplerType=SAMPLERTYPE_Masks, texture=/Game/Global/Materials/MaterialGym/Materials/Glitter/Textures/T_Glitter_TA}
  ) {samplerType=SAMPLERTYPE_Masks, texture=&Mash_Noise}
  n29 = Multiply(A=@n28.G, B=ScalarParameter() {defaultValue=1, parameterName=Micro_NoiseStrength}) {constB=0.1}
  NamedRerouteDeclaration(None=Add(
    A=NamedRerouteDeclaration(None=@n23) {name=Amp_Noise, nodeColor=<rgba 0 0.0625 1 1>}
    B=NamedRerouteDeclaration(None=@n29) {name=Micro_Noise, nodeColor=<rgba 0.3828 1 0 1>}
   )) {name=OverallNoise, nodeColor=<rgba 0.0781 0 1 1>}
  n30 = VertexColor()
  n31 = Constant()
  n32 = If(A=@n30.G, "A > B"=Constant() {r=0.2}, "A < B"=@n31) {constB=0.5}
  n33 = If(A=@n30.R, "A > B"=Constant() {r=0.4}, "A < B"=@n31) {constB=0.5}
  n34 = Constant() {r=1}
  NamedRerouteDeclaration(None=Max(A=Max(A=@n32, B=@n33), B=If(A=@n30.B, "A > B"=@n34, "A < B"=@n31) {constB=0.5})) {name=BeamMasks, nodeColor=<rgba 0.2422 1 0 1>}
  n35 = ScalarParameter() {defaultValue=20, parameterName=Electricity_EmissiveMult_Min}
  n36 = Subtract(A=@n5, B=ScalarParameter() {defaultValue=0.5, parameterName=Amp_Subtraction})
  n37 = Multiply(A=VectorParameter() {parameterName=BeamColor, defaultValue=<rgba 0.8 0 1 1>}.RGB, B=Saturate(None=Multiply(A=@n36) {constB=5}))
  NamedRerouteDeclaration(None=EyeAdaptationInverse(LightValueInput=Multiply(A=@n35, B=@n37), AlphaInput=ScalarParameter() {defaultValue=0.1, parameterName=EyeAdaptionInverse}).EyeAdaptationInverse) {name=Emissive, nodeColor=<rgba 1 0.8906 0 1>}
  NamedRerouteUsage()
  NamedRerouteUsage()
  NamedRerouteUsage()
  OneMinus(None=@n28.R)
  TextureSample(UVs=@n11) {samplerType=SAMPLERTYPE_LinearGrayscale, texture=/Game/Effects/Fort_Effects/Emotes/UltraLight/Texture/T_DirectionalNoise_Greyscale}
 ]
}
